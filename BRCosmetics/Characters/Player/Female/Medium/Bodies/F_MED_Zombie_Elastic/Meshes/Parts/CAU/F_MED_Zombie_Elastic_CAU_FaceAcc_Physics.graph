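# Physics asset for the CAU FaceAcc part of the medium female Zombie Elastic character, covering its dynamic hair bones.
# Small capsule bodies on nine numbered hair bones and the fl/fr pair surround a larger head capsule.
# Each hair bone is constrained directly to the head rather than chained, with tight twist limits.
# Numbered bones get wide swing cones, while the fl/fr pair is restricted to narrow, mostly single-axis swing.
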PhysicsAsset /BRCosmetics/Characters/Player/Female/Medium/Bodies/F_MED_Zombie_Elastic/Meshes/Parts/CAU/F_MED_Zombie_Elastic_CAU_FaceAcc_Physics
bodies {
 dyn_hair_1 [{shapeType=Capsule, length=1, radius=1, shapeName=dyn_hair_1_capsule, center=<1.1568 0 0>, rotation=<rot 0 -90 90>}]
 dyn_hair_2 [{shapeType=Capsule, length=1.75, radius=1, shapeName=dyn_hair_2_capsule, center=<1.1568 0 0>, rotation=<rot 0 -90 90>}]
 dyn_hair_3 [{shapeType=Capsule, length=1, radius=1.25, shapeName=dyn_hair_3_capsule, center=<1.1568 0 0>, rotation=<rot 0 -90 90>}]
 dyn_hair_4 [{shapeType=Capsule, length=1, radius=1, shapeName=dyn_hair_4_capsule, center=<1.1568 0 0>, rotation=<rot 0 -90 90>}]
 dyn_hair_5 [{shapeType=Capsule, length=1.25, radius=2.75, shapeName=dyn_hair_5_capsule, center=<2.9913 0 0>, rotation=<rot 0 -90 90>}]
 dyn_hair_6 [{shapeType=Capsule, length=1, radius=1.25, shapeName=dyn_hair_6_capsule, center=<1.2831 0 0>, rotation=<rot 0 -90 90>}]
 dyn_hair_7 [{shapeType=Capsule, length=1, radius=1, shapeName=dyn_hair_7_capsule, center=<1.1568 0 0>, rotation=<rot 0 -90 90>}]
 dyn_hair_8 [{shapeType=Capsule, length=1, radius=1.25, shapeName=dyn_hair_8_capsule, center=<0.7957 0 0>, rotation=<rot 0 -90 90>}]
 dyn_hair_9 [{shapeType=Capsule, length=1.5, radius=1, shapeName=dyn_hair_9_capsule, center=<1.1568 0 0>, rotation=<rot 0 -90 90>}]
 dyn_hair_fl [{shapeType=Capsule, length=1, radius=1.5, shapeName=dyn_hair_fl_capsule, center=<1.4043 0 0>, rotation=<rot 90 0 0>}]
 dyn_hair_fr [{shapeType=Capsule, length=1, radius=1.5, shapeName=dyn_hair_fr_capsule, center=<1.4043 0 0>, rotation=<rot 90 0 0>}]
 head [{shapeType=Capsule, length=10.7309, radius=7.8932, rotation=<rot 90 0 0>}]
}
constraints [
 {bone1Name=dyn_hair_1, bone2Name=head, swing1LimitDegrees=45, swing2LimitDegrees=45, twistLimitDegrees=2}
 {bone1Name=dyn_hair_2, bone2Name=head, swing1LimitDegrees=45, swing2LimitDegrees=45, twistLimitDegrees=2}
 {bone1Name=dyn_hair_3, bone2Name=head, swing1LimitDegrees=45, swing2LimitDegrees=45, twistLimitDegrees=2}
 {bone1Name=dyn_hair_4, bone2Name=head, swing1LimitDegrees=45, swing2LimitDegrees=45, twistLimitDegrees=2}
 {bone1Name=dyn_hair_5, bone2Name=head, swing1LimitDegrees=60, swing2LimitDegrees=60, twistLimitDegrees=2}
 {bone1Name=dyn_hair_6, bone2Name=head, swing1LimitDegrees=45, swing2LimitDegrees=45, twistLimitDegrees=2}
 {bone1Name=dyn_hair_7, bone2Name=head, swing1LimitDegrees=45, swing2LimitDegrees=45, twistLimitDegrees=2}
 {bone1Name=dyn_hair_8, bone2Name=head, swing1LimitDegrees=45, swing2LimitDegrees=45, twistLimitDegrees=2}
 {bone1Name=dyn_hair_9, bone2Name=head, swing1LimitDegrees=45, swing2LimitDegrees=45, twistLimitDegrees=2}
 {bone1Name=dyn_hair_fr, bone2Name=head, swing1LimitDegrees=20, swing2LimitDegrees=2, twistLimitDegrees=2}
 {bone1Name=dyn_hair_fl, bone2Name=head, swing1LimitDegrees=20, swing2LimitDegrees=2, twistLimitDegrees=2}
]
